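; War Effort funding outcome proxy actor that spawns a given actor class in its place, then destroys itself.
; SpawnActorAtProxy builds the transform via GetSpawnTransform, which takes the proxy's own location, rotation or scale wherever the override is left at default.
; With SpawnOnGround, a downward object trace at the proxy replaces that location, without checking for a hit.
; The UseOverrideAsOffset parameter is unused.
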(event Custom|SpawnActorAtProxy
  (params (ActorClass :type Actor%20Class%20Reference) (SpawnOnGround :type Boolean) (TransformOverride :type Transform) (UseOverrideAsOffset :type Boolean))
  (bind _outtransform (CallFunction|GetSpawnTransform TransformOverride))
  (if SpawnOnGround
    (bind (_returnvalue _location) (CallFunction|FindGroundLocation))
    (Game|SpawnActorfromClass ActorClass (Math|Transform|MakeTransform _location (.rotation _outtransform) (.scale _outtransform)) "AlwaysSpawn" "SelectDefaultAtRuntime" 0 0)
    (else
      (Game|SpawnActorfromClass ActorClass _outtransform "AlwaysSpawn" "SelectDefaultAtRuntime" 0 0)))
  (Actor|DestroyActor))

(fn FindGroundLocation
  (bind (_outhit _returnvalue) (Collision|LineTraceForObjects (+ (Transformation|GetActorLocation) (Math|Vector|MakeVector 0.0 0.0 (Variables|Default|GetGroundTraceHeightOffset))) (- (+ (Transformation|GetActorLocation) (Math|Vector|MakeVector 0.0 0.0 (Variables|Default|GetGroundTraceHeightOffset))) (Math|Vector|MakeVector 0.0 0.0 (Variables|Default|GetGroundTraceLength))) (Utilities|Array|MakeArray "ObjectTypeQuery1")))
  (return _returnvalue (.Location (Collision|BreakHitResult _outhit))))

(fn GetSpawnTransform
  (params (TransformOverride :type Transform))
  (bind (_returnvalue_location _returnvalue_rotation _returnvalue_scale) (Math|Transform|BreakTransform (Transformation|GetActorTransform)))
  (return (Math|Vector|SelectVector _returnvalue_location (.location TransformOverride) (Math|Vector|EqualExactly(Vector) (.location TransformOverride))) (Math|Rotator|SelectRotator _returnvalue_rotation (.rotation TransformOverride) (Utilities|Operators|Equal(==) :A (.rotation TransformOverride))) (Math|Vector|SelectVector _returnvalue_scale (.scale TransformOverride) (Math|Vector|EqualExactly(Vector) (.scale TransformOverride) "1.000000,1.000000,1.000000"))))

(fn ConstructionScript)
