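; Base War Effort funding station where players donate toward initiative options, with empty hooks for child classes.
; Donations are scaled by a hotfix multiplier and broadcast, with reward loot tossed from a socket.
; Replicated Stage selects removal, dated TBYB display, fundraising or funded visuals, with GameplayCues for feedback.
; On funding, the server resolves the initiative through the funding manager's config and passes an outcome object to the outcome handler component.
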(event EventBeginPlay
  ;[1 1 1] BeginPlay
  (|CallParentFunction)
  (Utilities|FlowControl|Sequence
    (:then_0
      (CallFunction|InitializeDecals)
      (CallFunction|InitializeTBYBMessage)
      (Networking|SwitchHasAuthority
        (:Authority
          (Class|FortAsyncActionAuthorityMatchReady|MatchReadyAsync))
        (:Remote
          (Variables|Default|SetInitial%20Relative%20Rotation (Variables|/Script/Engine.SceneComponent|GetRelativeRotation (Variables|Debug|GetStaticMeshComponent))))))))

(event Custom|FundraisingState)

(event Custom|NonexistentState
  ;[1 1 1] Sates to override in childs
  (Actor|DestroyActor))

(event Custom|OnFundsAdded_Feedback
  ;;[1 1 1] JLavender: Called when funds are added by players, but it does not reach the funding goal
  ;[1 1 1] Funds Added handlers
  (Networking|CanExecuteCosmeticEvents
    (:True
      (GameplayCue|ExecuteGameplayCue%28Burst%29Local self (Variables|GameplayCues|GetGCN_FundsAdded) (Ability|GameplayCue|MakeGameplayCueParameters 0.0)))))

(event Custom|EdgeServerShowFunding
  ;[1 1 1] Show funding visual feedback
  (CallFunction|FundsAdded))

(event Custom|FundsAdded
  (bind _countloottovend (Variables|Default|GetCountLootToVend))
  (bind _returnvalue (> _countloottovend 0))
  ;;[1 1 1] JLavender: Called when funds are added by players, but it does not reach the funding goal
  ;[1 1 1] Funds Added handlers
  (if _returnvalue
    (while _returnvalue
      (Class|AthenaFunctionLibrary|SpawnLootfromTable (Variables|Default|GetDontationRewardLootName) 1 -1 (Transformation|GetSocketLocation (Variables|Debug|GetStaticMeshComponent) "Drop_Loot") (Variables|Default|GetLootSpawnDirection) 0 true true true -1 "Tossed" "ItemSpawner")
      (Math|Integer|DecrementInt (Variables|Default|GetCountLootToVend)))
    (CallFunction|OnFundsAddedFeedback)))

(event Custom|FundedState
  (params (Choice :type Integer) (IsInitialFund :type Boolean))
  ;[1 1 1] On Funded - outcome handling
  ;;[1 1 1] FundedState - called by Funding Manager when a funding station has reached its funding limit. By default, constructs the data object containing the pertinent info about the outcome of funding - child classes will need to interpret and act on this data by overriding DoFundedOutcome
  (if (Networking|IsServer)
    (if (!= (Variables|Default|GetStage) 3)
      (bind (_initiativefound _initiative) (Class|WarEffortConfigData|GetInitiativebyTag (Variables|/WarEffort/MeshNetwork/B_WarEffort_FundingManager.B_WarEffort_FundingManager_C|GetWarEffortConfigData (Variables|Default|GetFundingManagerRef)) (Variables|Initiatives|GetActiveInitiative)))
      (if _initiativefound
        (Variables|Default|SetFundedChoice Choice)
        (Variables|Default|SetFundedInitiative _initiative)
        (bind _output_get (Variables|Default|SetIsInitialFund IsInitialFund))
        (Variables|Default|SetStage 3)
        (CallFunction|BroadcastOutcometoOutcomeHandler (Utilities|Array|Get(aref) (.Options_6_F30346D942E5CB445FA46C93CAB109F2 (Utilities|Struct|BreakFWarEffortInitiativeConfig _initiative)) Choice) _output_get)))))

(event Custom|DisableInteraction
  (Components|DestroyComponent (Variables|Components|GetNonPlayerConversationComponent))
  (Variables|Interaction|SetbAllowInteract false))

(event Custom|StationHandleFunded
  (params (Choice :type Integer) (Initiative :type FWar%20Effort%20Initiative%20Config%20Structure))
  ;[1 1 1] On Funded - outcome handling
  ;;[1 1 1] Handle funding outcomes that pertain to the station
  (CallFunction|DisableInteraction)
  (Cosmetic|SetDecalsfromChoice Choice)
  (CallFunction|FundingAchieved)
  (Cosmetic|UpdateBillboardonFunded Choice Initiative)
  (CallFunction|HideDonationBox)
  (if (Variables|Default|GetDestroyOnFund)
    (Gameplay|SilentDie false)))

(event Custom|BroadcastOutcomeToOutcomeHandler
  (params (FWarEffortOptionConfig :type FWar%20Effort%20Option%20Config%20Structure%20%28by%20ref%29) (IsInitialFund :type Boolean))
  ;[1 1 1] On Funded - outcome handling
  ;;[1 1 1] Send outcome data to outcome handler component
  (bind _returnvalue (Game|ConstructObjectfromClass (Utilities|Struct|BreakFWarEffortOptionConfig FWarEffortOptionConfig)))
  (bind _outcomehandlercomponent (Getters|GetOutcomeHandlerComponent))
  (Utilities|IsValid _outcomehandlercomponent
    (:Is%20Valid
      (Class|BWarEffortOutcomeHandlerComponent|DoFundedOutcome _outcomehandlercomponent _returnvalue IsInitialFund))))

(event Custom|DoFundedOutcome
  (params (OutcomeData :type War%20Effort%20Outcome%20Data%20Object%20Reference)))

(event Custom|FundingAchieved
  ;[1 1 1] On Funded - outcome handling
  ;;[1 1 1] JLavender: Called when funds are added and it meets funding goal
  (bind _result (Math|Integer|IncrementInt (Variables|Default|GetUpdateFundsAchieved))
    (:%20%20
      (Variables|Default|SetUpdateFundsAchieved _result))))

(event Custom|HideDonationBox
  (bind _staticmeshcomponent (Variables|Debug|GetStaticMeshComponent))
  ;[1 1 1] On Funded - outcome handling
  (Rendering|SetVisibility _staticmeshcomponent false false)
  (Collision|SetCollisionEnabled _staticmeshcomponent "NoCollision"))

(event Custom|UpdateSmokeValues)

(event Custom|OnFundingAchieved_Feedback
  (bind (_stationtypeint _smokeloc _smokesize) (Cosmetic|FXGetSmokeCoverCueParams))
  ;[1 1 1] On Funded - outcome handling
  ;;[1 1 1] JLavender: Called when funds are added and it meets funding goal
  (GameplayCue|ExecuteGameplayCue%28Burst%29Local self (Variables|GameplayCues|GetGCN_FundingAchieved) (Ability|GameplayCue|MakeGameplayCueParameters :NormalizedMagnitude (Math|Conversions|ToFloat(Integer) _stationtypeint) :RawMagnitude (Math|Conversions|ToFloat(Integer) _smokesize) :Location _smokeloc :Normal "0,0,0" :GameplayEffectLevel 1 :AbilityLevel 1 :bReplicateLocationWhenUsingMinimalRepProxy false)))

(fn AddToStationFund
  (params (FundingAmount :type Integer64) (FundChoice :type Integer))
  (Math|Integer|IncrementInt (Variables|Default|GetCountLootToVend)
    (:%20%20
      (Default|CallOnFundingReceived (Math|Float|TruncatetoInteger64 (* (Math|Conversions|ToInteger(Integer64) FundingAmount) (ScalableFloat|GetValueatLevel (Variables|Hotfix|GetHF_FundMultiplier)))) (Variables|Default|GetFundingStationNum) FundChoice))))

(fn AssignedToAnyInitiative
  (params (InitiativeTags :type Gameplay%20Tag%20Container%20Structure%20%28by%20ref%29))
  (bind (_hasactiveinitiative _initiativetag) (Initiative|FindActiveInitiative InitiativeTags))
  (return _hasactiveinitiative _initiativetag))

(fn AssignedToInitiative
  (params (InitiativeTag :type Gameplay%20Tag%20Structure))
  (return (GameplayTags|HasTag (Variables|Initiatives|GetInitiatives) InitiativeTag true)))

(fn DestroyPreviewConstructionActors)

(fn EnableTBYB
  (if (not (ScalableFloat|GetValueasBool (Variables|Hotfix|GetHF_IgnoreCalendarEvent)))
    (CallFunction|SetDecalDate)
    (Rendering|SetVisibility (Variables|Default|GetDateDecal) true false)
    (CallFunction|DisableInteraction)
    (CallFunction|HideDonationBox)))

(fn FindActiveInitiative
  (params (ActiveInitiatives :type Gameplay%20Tag%20Container%20Structure))
  (for _array_element (GameplayTags|BreakGameplayTagContainer ActiveInitiatives)
    (if (GameplayTags|HasTag (Variables|Initiatives|GetInitiatives) _array_element true)
      (return true _array_element)))
  (return false 0))

(fn FX_GetSmokeCoverCueParams
  (return 0 (Transformation|GetActorLocation) 0))

(fn GetChoiceHotfix
  (return (ScalableFloat|GetValueasInteger (Variables|Hotfix|GetHF_FundingChoice))))

(fn GetCurrentFunds
  (bind (_initiativefound _options) (Getters|GetInitiativeOptions))
  (for (_ _array_index) _options
    (bind _output (Funding|Get|GetCurrentFundsforInitiative (Variables|Default|GetFundingManagerRef) (Variables|Initiatives|GetActiveInitiative) _array_index self))
    (Utilities|Array|Add (LocalVariables|GetFunds) _output))
  (return (LocalVariables|GetFunds)))

(fn GetEnabledHotfix
  (return (ScalableFloat|GetValueasBool (Variables|Hotfix|GetHF_StationEnabled))))

(fn GetFundedHotfix
  (return (ScalableFloat|GetValueasBool (Variables|Hotfix|GetHF_Funded))))

(fn GetFundingStationIndex
  (return (Variables|Default|GetFundingStationNum)))

(fn GetIndexForInitiative
  (params (InitiativeTag :type Gameplay%20Tag%20Structure%20%28by%20ref%29))
  (bind (_value _returnvalue) (Utilities|Map|Find (Variables|Initiatives|GetInitiativeIndexMap) InitiativeTag))
  (return _returnvalue _value))

(fn GetInitiativeOptions
  (bind (_initiativefound _initiative) (Class|WarEffortConfigData|GetInitiativebyTag (Variables|/WarEffort/MeshNetwork/B_WarEffort_FundingManager.B_WarEffort_FundingManager_C|GetWarEffortConfigData (Variables|Default|GetFundingManagerRef)) (Variables|Initiatives|GetActiveInitiative)))
  (return _initiativefound (.Options_6_F30346D942E5CB445FA46C93CAB109F2 (Utilities|Struct|BreakFWarEffortInitiativeConfig _initiative))))

(fn GetItemIcons
  (bind (_initiativefound _options) (Getters|GetInitiativeOptions))
  (if _initiativefound
    (for _array_element _options
      (Utilities|Array|Add (LocalVariables|GetItemIcons) (Utilities|Struct|BreakFWarEffortOptionConfig _array_element)))
    (return (LocalVariables|GetItemIcons))))

(fn GetItemNames
  (bind (_initiativefound _options) (Getters|GetInitiativeOptions))
  (if _initiativefound
    (for _array_element _options
      (Utilities|Array|Add (LocalVariables|GetNames) (.DisplayName_22_E10681B54C5D823F36E65599A51952E5 (Utilities|Struct|BreakFWarEffortOptionConfig _array_element))))
    (return (LocalVariables|GetNames))))

(fn GetItemSpawnPoint
  (return (Transformation|GetSocketLocation (Variables|Debug|GetStaticMeshComponent) "Drop_Loot")))

(fn GetLootSpawnPosition
  (return (Transformation|GetSocketLocation (Variables|Debug|GetStaticMeshComponent) "Drop_Loot")))

(fn GetMaxFunds
  (bind (_fundinglimitfound _fundinglimit) (Funding|Get|GetFundingTargetbyInitiative (Variables|Default|GetFundingManagerRef) (Variables|Initiatives|GetActiveInitiative)))
  (return _fundinglimit))

(fn GetOutcomeHandlerComponent
  (bind _asb_war_effort_outcome_handler_component (Utilities|Casting|CastToB_WarEffort_OutcomeHandlerComponent (Actor|GetComponentbyClass (Variables|Default|GetOutcomeHandlerComponentBaseClass))))
  (return _asb_war_effort_outcome_handler_component))

(fn InitializeDecals
  (bind _output_get (Variables|Display|SetFundedDecals (Utilities|Array|MakeArray (Variables|Default|GetDecal_Left) (Variables|Default|GetDecal_Right))))
  (for _array_element _output_get
    (Rendering|SetVisibility _array_element false false)))

(fn InitializeTBYBMessage
  (Rendering|SetVisibility (Variables|Default|GetDateDecal) false false))

(fn IsFunded
  (params (Choice :type Integer))
  (bind _fundingmanagerref (Variables|Default|GetFundingManagerRef))
  (Utilities|IsValid _fundingmanagerref
    (:Is%20Valid
      (bind _returnvalue (Class|BWarEffortFundingManager|IsFundingStationFunded _fundingmanagerref self Choice))
      (bind _returnvalue_2 (CallFunction|GetFundedHotfix))
      (return (or _returnvalue_2 _returnvalue)))
    (:Is%20Not%20Valid
      (return false))))

(fn OnRep_BillboardMaterial
  (Rendering|Material|SetMaterial (Variables|Default|GetBillBoard) 1 (Variables|Default|GetBillboardMaterial)))

(fn OnRep_Stage
  (switch int (Variables|Default|GetStage)
    (:0
      (CallFunction|NonexistentState))
    (:1
      (CallFunction|EnableTBYB))
    (:2
      (CallFunction|FundraisingState))
    (:3
      (CallFunction|StationHandleFunded (Variables|Default|GetFundedChoice) (Variables|Default|GetFundedInitiative)))))

(fn OnRep_UpdateFundAchieved
  (CallFunction|UpdateSmokeValues)
  (Networking|CanExecuteCosmeticEvents
    (:True
      (CallFunction|OnFundingAchievedFeedback))))

(fn OnRep_UpdateFundsAdded
  (Networking|CanExecuteCosmeticEvents
    (:True
      (CallFunction|OnFundsAddedFeedback))))

(fn PlayServiceUsedEffects
  (params (PlayerController :type Fort%20Player%20Controller%20Object%20Reference))
  (|CallParentFunction PlayerController)
  (Utilities|IsValid PlayerController
    (:Is%20Valid
      (bind _asathena_player_controller (Utilities|Casting|CastToAthena_PlayerController PlayerController))
      (EventDispatchers|CallOnClientDonationSuccessDelegate _asathena_player_controller))))

(fn SetActive
  (params (Active :type Boolean))
  (if (not Active)
    (Actor|DestroyActor)))

(fn SetActiveInitiative
  (params (ActiveInitiatives :type Gameplay%20Tag%20Container%20Structure))
  (bind (_hasactiveinitiative _initiativetag) (Initiative|FindActiveInitiative ActiveInitiatives))
  (if _hasactiveinitiative
    (bind _output_get (Variables|Initiatives|SetActiveInitiative _initiativetag))
    (bind (_initiativefound _initiative) (Class|WarEffortConfigData|GetInitiativebyTag (Variables|/WarEffort/MeshNetwork/B_WarEffort_FundingManager.B_WarEffort_FundingManager_C|GetWarEffortConfigData (Variables|Default|GetFundingManagerRef)) _output_get))
    (if _initiativefound
      (Variables|Default|SetBillboardMaterial (Utilities|Struct|BreakFWarEffortInitiativeConfig _initiative))
      (return true)
      (else
        (return false)))
    (else
      (return false))))

(fn SetDecalDate
  (bind _day (Variables|DateDecal|GetDay))
  (bind (_1st_digit _2nd_digit) (CallFunction|SplitIntegertoIndividualDigits _day))
  (bind _month (Variables|DateDecal|GetMonth))
  (bind (_1st_digit_2 _2nd_digit_2) (CallFunction|SplitIntegertoIndividualDigits _month))
  (Networking|CanExecuteCosmeticEvents
    (:True
      (Utilities|IsValid (Variables|DateDecal|GetDateDecalMID)
        (:Is%20Valid
          _)
        (:Is%20Not%20Valid
          (bind _returnvalue (Rendering|Components|Decal|CreateDynamicMaterialInstance (Variables|Default|GetDateDecal)))
          (Variables|DateDecal|SetDateDecalMID _returnvalue)))
      (Rendering|Material|SetScalarParameterValue (Variables|DateDecal|GetDateDecalMID) "Char0" _1st_digit_2 :!class /Script/Engine.MaterialInstanceDynamic)
      (Rendering|Material|SetScalarParameterValue (Variables|DateDecal|GetDateDecalMID) "Char1" _2nd_digit_2 :!class /Script/Engine.MaterialInstanceDynamic)
      (Rendering|Material|SetScalarParameterValue (Variables|DateDecal|GetDateDecalMID) "Char2" _1st_digit :!class /Script/Engine.MaterialInstanceDynamic)
      (Rendering|Material|SetScalarParameterValue (Variables|DateDecal|GetDateDecalMID) "Char3" _2nd_digit :!class /Script/Engine.MaterialInstanceDynamic)
      (Rendering|Material|SetVectorParameterValue (Variables|DateDecal|GetDateDecalMID) "TextColor" (Variables|DateDecal|GetTextColor) :!class /Script/Engine.MaterialInstanceDynamic)
      (Rendering|Material|SetVectorParameterValue (Variables|DateDecal|GetDateDecalMID) "OutlineColor" (Variables|DateDecal|GetOutlineColor) :!class /Script/Engine.MaterialInstanceDynamic))))

(fn SetDecalsFromChoice
  (params (Choice :type Integer))
  (bind _decal_right (Variables|Default|GetDecal_Right))
  (bind _decal_left (Variables|Default|GetDecal_Left))
  (Rendering|Components|Decal|SetDecalMaterial _decal_left (Utilities|Select :Index Choice :Option%200 (Variables|Display|GetWinnerDecalMaterial) :Option%201 (Variables|Display|GetLoserDecalMaterial)))
  (Rendering|SetVisibility _decal_left true false)
  (Rendering|Components|Decal|SetDecalMaterial _decal_right (Utilities|Select :Index Choice :Option%200 (Variables|Display|GetLoserDecalMaterial) :Option%201 (Variables|Display|GetWinnerDecalMaterial)))
  (Rendering|SetVisibility _decal_right true false))

(fn SetDressActors
  (Variables|Default|SetArrayOfSetDressActors (Utilities|Array|MakeArray 0 0 0 0 0 0)))

(fn SetIndexForInitiative
  (params (InitiativeTag :type Gameplay%20Tag%20Structure) (Index :type Integer))
  (Utilities|Map|Add (Variables|Initiatives|GetInitiativeIndexMap) InitiativeTag Index))

(fn SplitIntegerToIndividualDigits
  (params (Int :type Integer))
  (return (Math|Conversions|ToFloat(Integer) (/ (- Int (Math|Integer|%25%28Integer%29 Int 10)) 10)) (Math|Conversions|ToFloat(Integer) (Math|Integer|%25%28Integer%29 Int 10))))

(fn UpdateBillboardOnFunded
  (params (Choice :type Integer) (Initiative :type FWar%20Effort%20Initiative%20Config%20Structure))
  (bind (_initiativetag_10_3a0f4cb0475942112d4ae38f877c4735 _initiativetypetag_14_11cacbab4f570f662fa860ac59cc19f3 _options_6_f30346d942e5cb445fa46c93cab109f2) (Utilities|Struct|BreakFWarEffortInitiativeConfig Initiative))
  (if (> (Utilities|Array|Length _options_6_f30346d942e5cb445fa46c93cab109f2) 1)
    (Variables|Default|SetFundedChoice Choice)))

(fn ConstructionScript
  (|CallParentFunction)
  (if (Variables|Display|GetRemoveProps)
    _))
