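# Translucent deferred decal material for War Effort stations that draws a five-character date from a vertical 11-glyph number atlas.
# Sphere masks on scaled U select each slot's V offset: Char0–Char3 drive four digits, while the middle slot uses the fixed last glyph.
# Derivatives from unshifted UVs drive mip selection, avoiding seams between slots.
# Atlas red and green channels are tinted by TextColor and OutlineColor, then summed.
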Material /WarEffort/Stations/Materials/M_WarEffor_Date
graph {
 properties {blendMode=BLEND_Translucent, materialDomain=MD_DeferredDecal, pixelDepthOffsetMode=PDOM_Legacy, usedShadingModels=Default Lit}
 nodes [
  n1 = Constant()
  n2 = ScalarParameter() {parameterName=Char0}
  n3 = NamedRerouteUsage()
  n4 = Constant() {r=0.5}
  n5 = Constant() {r=1}
  n6 = ScalarParameter() {defaultValue=1, parameterName=Char1}
  n7 = NamedRerouteUsage()
  n8 = Constant() {r=1.5}
  n9 = Constant() {r=0.5}
  n10 = Constant() {r=1}
  n11 = Add(
   A=Multiply(A=Divide(A=@n2) {constB=11}, B=SphereMask(A=@n3, B=@n4, Radius=@n4, Hardness=@n5))
   B=Multiply(A=Divide(A=@n6) {constB=11}, B=SphereMask(A=@n7, B=@n8, Radius=@n9, Hardness=@n10))
  )
  n12 = Multiply(
   A=Divide(A=ScalarParameter() {defaultValue=2, parameterName=Char2}) {constB=11}
   B=SphereMask(A=NamedRerouteUsage(), B=Constant() {r=3.5}, Radius=Constant() {r=0.5}, Hardness=Constant() {r=1})
  )
  n13 = Constant() {r=0.9091}
  n14 = SphereMask(A=NamedRerouteUsage(), B=Constant() {r=2.5}, Radius=Constant() {r=0.5}, Hardness=Constant() {r=1})
  n15 = ScalarParameter() {defaultValue=7, parameterName=Char3}
  n16 = NamedRerouteUsage()
  n17 = Constant() {r=4.5}
  n18 = Constant() {r=0.5}
  n19 = Constant() {r=1}
  n20 = Add(
   A=Add(A=Add(A=@n11, B=@n12), B=Multiply(A=@n13, B=@n14))
   B=Multiply(A=Divide(A=@n15) {constB=11}, B=SphereMask(A=@n16, B=@n17, Radius=@n18, Hardness=@n19))
  )
  n21 = NamedRerouteUsage()
  n22 = TextureSample!(
   UVs=Add(A=NamedRerouteUsage(), B=AppendVector(A=@n1, B=@n20))
   Tex
   "DDX(UVs)"=DDX(Value=@n21)
   "DDY(UVs)"=DDY(Value=@n21)
   Apply View MipBias
  ) {mipValueMode=TMVM_Derivative, samplerType=SAMPLERTYPE_LinearColor, texture=/WarEffort/Stations/Materials/Textures/T_WarEffort_Numbers}
  Add(
   A=Multiply(A=@n22.R, B=VectorParameter() {parameterName=TextColor, defaultValue=<rgba 0.8698 0.6385 0.4871 0>}.RGB)
   B=Multiply(A=@n22.G, B=VectorParameter() {parameterName=OutlineColor, defaultValue=<rgba 0.0599 0.0141 0.3872 0>}.RGB)
  )
  Add(A=@n22.R, B=@n22.R)
  n23 = TextureCoordinate() {uTiling=5, vTiling=0.0909}
  NamedRerouteDeclaration(None=MaterialFunctionCall(Float2=@n23) {materialFunction=&BreakOutFloat2Components}.R) {name=Scaled.U, nodeColor=<rgba 0 0.6016 1 1>}
  NamedRerouteDeclaration(None=@n23) {name=Scaled.UV, nodeColor=<rgba 1 0 0.3281 1>}
 ]
}
