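# Deferred decal material that draws a tinted check-mark icon, apparently for War Effort station boards.
# A Color vector parameter, green by default, is multiplied by the alpha of a UI check icon texture to form the tinted shape.
# It uses alpha-composite blending with Default Lit shading; a Lerp and two constants are also present with no inputs wired into the Lerp.
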Material /WarEffort/Stations/Materials/M_WarEffort_BoardDecal_Check
graph {
 properties {blendMode=BLEND_AlphaComposite, materialDomain=MD_DeferredDecal, pixelDepthOffsetMode=PDOM_Legacy, usedShadingModels=Default Lit}
 nodes [
  Constant()
  Constant() {r=0.5}
  LinearInterpolate()
  Multiply(
   A=VectorParameter() {parameterName=Color, defaultValue=<rgba 0.1333 1 0 0>}.RGB
   B=TextureSample() {texture=/Game/UI/Foundation/Textures/Icons/Checks/T-Icon-Check-128}.A
  )
 ]
}
